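; Construction script for an air duct grate prop in an Apollo special map, which checks the orientation of its static mesh.
; It reads the static mesh component's relative rotation and tests absolute roll, then pitch, against a small tolerance.
; Both branch bodies appear empty, and the script also contains two billboard component additions, possibly intended as markers for tilted placement.
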(fn ConstructionScript
  (bind _staticmeshcomponent (Variables|Debug|GetStaticMeshComponent))
  (bind _returnvalue (Transformation|GetRelativeTransform _staticmeshcomponent))
  (bind _rotation (.rotation _returnvalue))
  (if (> (Math|Float|Absolute(Float) (.roll _rotation)) 0.1)
    _
    (elif (> (Math|Float|Absolute(Float) (.pitch _rotation)) 0.1)
      _))
  (AddComponent|NODE_AddBillboardComponent-0)
  (AddComponent|NODE_AddBillboardComponent-1))
